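# Sound cue for corruption core lightning, built from Apollo storm, enemy portal and cube zap waves.
# A random close-lightning wave mixes with a modulated portal lightning wave, and an enveloper fades that pair out by about two seconds.
# A cube zap layer mixes in outside the envelope.
# Overridden attenuation adds distance low-pass filtering and listener focus, alongside a core concurrency set and a distance-curved distant reverb send.
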SoundCue /CorruptionGameplay/Sounds/Corruption/Corruption_Core_Lightning_Cue
root=Mixer_0
nodes [
 {
  id=Enveloper_0
  children [Mixer_1]
  properties {
   pitchCurve {editorCurveData {keys [{time=0}]}}
   volumeCurve {editorCurveData {keys [<at 0 1>,<at 0.7091 1>,<at 1.121 0.5667>,<at 1.9767 0>]}}
  }
 }
 {id=Mixer_0, children [Enveloper_0,WavePlayer_7], properties {inputVolume [1,1]}}
 {id=Mixer_1, children [Modulator_1,Random_0], properties {inputVolume [1,1]}}
 {id=Modulator_1, children [WavePlayer_6]}
 {id=Random_0, children [WavePlayer_0,WavePlayer_1,WavePlayer_2], properties {weights [1,1,1]}}
 {id=WavePlayer_0, properties {soundWaveAssetPtr=/Game/Athena/Sounds/Shield/ApolloStorm/Apollo_Lightning_Close_02}}
 {id=WavePlayer_1, properties {soundWaveAssetPtr=/Game/Athena/Sounds/Shield/ApolloStorm/Apollo_Lightning_Close_04}}
 {id=WavePlayer_2, properties {soundWaveAssetPtr=/Game/Athena/Sounds/Shield/ApolloStorm/Apollo_Lightning_Close_08}}
 {id=WavePlayer_6, properties {soundWaveAssetPtr=/Game/Sounds/Fort_GamePlay_Sounds/Portal/EnemyPortal_Lightning06}}
 {id=WavePlayer_7, properties {soundWaveAssetPtr=/Game/Athena/Sounds/Events/Cube/Cube_AmbZap_01}}
]
properties {
 bOverrideAttenuation=true
 maxDistance=2900
 priority=0.5
 soundClassObject=/Game/Sounds/FortSoundClassesAndModes/Fort_AGV
 attenuationOverrides {
  bAttenuateWithLPF=true
  bEnableFocusInterpolation=true
  bEnableListenerFocus=true
  falloffDistance=2500
  focusReleaseInterpSpeed=0.5
  lPFFrequencyAtMax=5000
  lPFRadiusMax=3000
  lPFRadiusMin=400
  nonFocusVolumeAttenuation=0.4
 }
 concurrencySet [&/Corruption_Core_Conc]
 soundSubmixSends [{
   maxSendDistance=2500
   minSendDistance=0
   minSendLevel=0
   sendLevel=0
   sendLevelControlMethod=CustomCurve
   soundSubmix=&Fortnite_Reverb_Distant
   customSendLevelCurve {editorCurveData {keys [<at 0 0>,<at 1 1>]}}
  }]
}
