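# Sound cue that plays a lightning strike, named for the Deimos spawn in Corruption gameplay.
# A mixer layers one of three close storm lightning waves, faded out by a volume envelope, with one of four Zapper lightning impact waves.
# Overridden attenuation adds distance low-pass and listener focus, with a curve-driven send to distant reverb.
# Playback is governed by the Corruption core concurrency set.
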SoundCue /CorruptionGameplay/Sounds/Corruption/Corruption_Deimos_Spawn_Lightning_Cue
root=Mixer_0
nodes [
 {
  id=Enveloper_0
  children [Random_0]
  properties {
   pitchCurve {editorCurveData {keys [{time=0}]}}
   volumeCurve {editorCurveData {keys [<at 0 1>,<at 0.4312 1>,<at 0.587 0.7501>,<at 0.9041 0.4186>,<at 1.5491 0>]}}
  }
 }
 {id=Mixer_0, children [Enveloper_0,Random_1], properties {inputVolume [1,1]}}
 {id=Random_0, children [WavePlayer_0,WavePlayer_1,WavePlayer_2], properties {weights [1,1,1]}}
 {id=Random_1, children [WavePlayer_3,WavePlayer_4,WavePlayer_5,WavePlayer_8], properties {weights [1,1,1,1]}}
 {id=WavePlayer_0, properties {soundWaveAssetPtr=/Game/Athena/Sounds/Shield/ApolloStorm/Apollo_Lightning_Close_02}}
 {id=WavePlayer_1, properties {soundWaveAssetPtr=/Game/Athena/Sounds/Shield/ApolloStorm/Apollo_Lightning_Close_04}}
 {id=WavePlayer_2, properties {soundWaveAssetPtr=/Game/Athena/Sounds/Shield/ApolloStorm/Apollo_Lightning_Close_08}}
 {id=WavePlayer_3, properties {soundWaveAssetPtr=/Game/Sounds/Fort_Monster_Zapper/Zapper_Weapon_Impact_Lightning_01}}
 {id=WavePlayer_4, properties {soundWaveAssetPtr=/Game/Sounds/Fort_Monster_Zapper/Zapper_Weapon_Impact_Lightning_02}}
 {id=WavePlayer_5, properties {soundWaveAssetPtr=/Game/Sounds/Fort_Monster_Zapper/Zapper_Weapon_Impact_Lightning_03}}
 {id=WavePlayer_8, properties {soundWaveAssetPtr=/Game/Sounds/Fort_Monster_Zapper/Zapper_Weapon_Impact_Lightning_04}}
]
properties {
 bOverrideAttenuation=true
 maxDistance=4000
 priority=0.5
 soundClassObject=/Game/Sounds/FortSoundClassesAndModes/Fort_AGV
 volumeMultiplier=0.85
 attenuationOverrides {
  bAttenuateWithLPF=true
  bEnableFocusInterpolation=true
  bEnableListenerFocus=true
  falloffDistance=2500
  focusReleaseInterpSpeed=0.5
  lPFFrequencyAtMax=5000
  lPFRadiusMax=3000
  lPFRadiusMin=400
  nonFocusVolumeAttenuation=0.4
  attenuationShapeExtents {x=1500}
 }
 concurrencySet [&/Corruption_Core_Conc]
 soundSubmixSends [{
   maxSendDistance=2500
   minSendDistance=0
   minSendLevel=0
   sendLevel=0
   sendLevelControlMethod=CustomCurve
   soundSubmix=&Fortnite_Reverb_Distant
   customSendLevelCurve {editorCurveData {keys [<at 0 0>,<at 1 1>]}}
  }]
}
